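# Post-process material, applied after depth of field, that makes the screen appear to collapse toward a world-space point, with an animated flare there.
# Scene-color UVs shift along the direction to the projected center, scaled by ray-traced sphere distance and ProgressCollapse, with radial masks darkening out-of-range areas.
# Layered, sine-flickered flare textures are added, controlled by a parameter collection.
# Leftover ghosting, twist-distortion and virtual-plane branches appear to be unused experiments.
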Material /Figment_S10/Packages/Fortnite_UniqueMaterials_1/PostProcess/PPM_HLODWPOpost
graph {
 properties {blendableLocation=BL_SceneColorAfterDOF, materialDomain=MD_PostProcess, pixelDepthOffsetMode=PDOM_Legacy, usedShadingModels=Default Lit}
 nodes [
  n1 = TextureCoordinate()
  n2 = VectorParameter() {parameterName=CollapseCenter, defaultValue=<rgba 66048 -43008 0 1>}
  n3 = ComponentMask(None=MaterialFunctionCall(World Position=@n2.RGB) {materialFunction=&TransformToClipSpace}.Screen aligned 0-1 UV) {g=true, r=true}
  n4 = TextureCoordinate()
  n5 = Constant2Vector()
  n6 = ScalarParameter() {defaultValue=18317732, parameterName=SphereRadius}
  n7 = MaterialFunctionCall!(
   Sphere Position=Add(A=AppendVector(A=@n5, B=@n6), B=@n2.RGB)
   Sphere Radius=@n6
   RayDirection=CameraVectorWS()
   World position=WorldPosition().XYZ
  ) {materialFunction=/Engine/Functions/Engine_MaterialFunctions02/RayTracedSphere}
  n8 = Distance(A=@n7.Intersection End World Position, B=@n2.RGB)
  n9 = Divide(
   A=Power(Base=@n8, Exp=ScalarParameter() {defaultValue=-0.37, parameterName=DistanceFalloff})
   B=ScalarParameter() {defaultValue=1040.9625, parameterName=Distance2}
  )
  n10 = Multiply(A=Normalize(VectorInput=Subtract(A=@n3, B=@n4)), B=@n9)
  n11 = Saturate(None=Power(
    Base=CollectionParameter() {collection=/Game/Packages/Fortnite_UniqueMaterials_1/PostProcess/MPC_PostProcessControl, parameterName=ProgressCollapse}
    Exp=ScalarParameter() {defaultValue=6, parameterName=ExponentialTime}
   ))
  n12 = Constant2Vector() {g=0.5, r=0.5}
  n13 = SphereMask(A=@n1, B=@n12, Radius=ScalarParameter() {defaultValue=0.5, parameterName=Radius}, Hardness=ScalarParameter() {parameterName=Hardness})
  n14 = ScalarParameter() {defaultValue=1.3679, parameterName=FallofRadiusAExp}
  n15 = Power(Base=@n11, Exp=ScalarParameter() {defaultValue=8, parameterName=ScreenEdgeDelay})
  n16 = LinearInterpolate(
   A=Constant()
   B=LinearInterpolate(A=Constant(), B=ScalarParameter() {defaultValue=-70300000256, parameterName=DistanceTest2}, Alpha=@n11)
   Alpha=LinearInterpolate(A=Power(Base=@n13, Exp=@n14), B=Constant() {r=1}, Alpha=Saturate(None=@n15))
  )
  n17 = CameraPositionWS()
  n18 = Add(A=@n1, B=Divide(A=Multiply(A=@n10, B=@n16), B=Distance(A=@n2.RGB, B=@n17)))
  n19 = SphereMask(
   A=@n18
   B=Constant2Vector() {g=0.5, r=0.5}
   Radius=ScalarParameter() {defaultValue=0.5, parameterName=Radius}
   Hardness=ScalarParameter() {parameterName=Hardness}
  )
  n20 = ScalarParameter() {defaultValue=1.5705, parameterName=FallofRadiusAExp2}
  n21 = Power(Base=OneMinus(None=@n11), Exp=ScalarParameter() {defaultValue=2, parameterName=ScreenEdgeDelay2})
  n22 = ComponentMask(None=SceneTexture(UVs=@n18) {sceneTextureId=PPI_PostProcessInput0}.Color) {b=true, g=true, r=true}
  n23 = Multiply(A=LinearInterpolate(A=Power(Base=@n19, Exp=@n20), B=Constant() {r=1}, Alpha=Saturate(None=@n21)), B=@n22)
  n24 = Constant3Vector() {constant=<rgba 2.67 4.6843 5 1>}
  n25 = Subtract(
   A=Subtract(
    A=ScreenPosition().ViewportUV
    B=MaterialFunctionCall(World Position=@n2.RGB) {materialFunction=&TransformToClipSpace}.Buffer UVs for Scene Textures
   )
   B=Constant2Vector() {g=-0.5, r=-0.5}
  )
  n26 = Constant2Vector() {g=1, r=1}
  n27 = MaterialFunctionCall(Float4=CollectionParameter() {collection=&MPC_PostProcessControl, parameterName=Flare_Scale}) {materialFunction=&BreakOutFloat4Components}
  n28 = Constant() {r=0.01}
  n29 = MaterialFunctionCall(
   X=Clamp(None=@n27.R, Min=@n28) {clampMode=CMODE_ClampMin, minDefault=0.01}
   Y=Clamp(None=@n27.G, Min=@n28) {clampMode=CMODE_ClampMin, minDefault=0.01}
   Z=Clamp(None=@n27.B, Min=@n28) {clampMode=CMODE_ClampMin, minDefault=0.01}
   A=Clamp(None=@n27.A, Min=@n28) {clampMode=CMODE_ClampMin, minDefault=0.01}
  ) {materialFunction=&MakeFloat4}
  n30 = ComponentMask(None=@n29.Result) {g=true, r=true}
  n31 = SphereMask(
   A=MaterialFunctionCall(UVs=@n25, Texture Scale=Multiply(A=@n26, B=@n30)) {materialFunction=&ScaleUVsByCenter}.UVs
   B=Constant() {r=0.5}
  ) {attenuationRadius=0.6, hardnessPercent=0}
  n32 = Multiply(A=Divide(A=Power(Base=@n31) {constExponent=4}) {constB=25}, B=Constant3Vector() {constant=<rgba 0.578 0.7602 1 0>})
  n33 = Multiply(A=Constant2Vector() {g=0.6, r=0.3}, B=@n30)
  n34 = Multiply(A=@n33, B=Constant2Vector() {g=0.8, r=3}) {constB=0.8}
  n35 = Time()
  n36 = Multiply(A=Add(A=@n35) {constB=5.7}) {constB=7}
  n37 = Sine(None=@n36) {period=0.155}
  n38 = Sine(None=@n36) {period=0.2}
  n39 = Multiply(A=Sine(None=@n36) {period=0.075}, B=Multiply(A=Multiply(A=@n37, B=@n38), B=Sine(None=@n36) {period=0.43}))
  n40 = LinearInterpolate(Alpha=@n39) {constA=0.95}
  n41 = Rotator(Coordinate=MaterialFunctionCall(UVs=@n25, Texture Scale=Multiply(A=@n34, B=@n40)) {materialFunction=&ScaleUVsByCenter}.UVs) {speed=-0.15}
  n42 = MaterialFunctionCall(UVs=@n25, Texture Scale=Multiply(A=@n33, B=@n40)) {materialFunction=&ScaleUVsByCenter}
  n43 = Add(
   A=Divide(A=TextureSample(UVs=@n41) {samplerSource=SSM_Clamp_WorldGroupSettings, texture=/Game/Effects/Fort_Effects/Textures/Flares/T_Flare_Spikeball}.RGB) {constB=3}
   B=TextureSample(UVs=Rotator(Coordinate=@n42.UVs) {speed=0.11}) {samplerSource=SSM_Clamp_WorldGroupSettings, texture=&T_Flare_Spikeball}.RGB
  )
  n44 = ComponentMask(None=@n29.Result) {a=true, b=true}
  n45 = Constant2Vector() {g=0.9, r=1.2}
  n46 = MaterialFunctionCall(UVs=@n25, Texture Scale=Multiply(A=Multiply(A=@n44, B=@n45), B=@n40)) {materialFunction=&ScaleUVsByCenter}
  n47 = TextureSample(UVs=@n46.UVs) {
   samplerSource=SSM_Clamp_WorldGroupSettings
   samplerType=SAMPLERTYPE_Grayscale
   texture=/Game/Effects/Fort_Effects/Textures/Flares/T_Cross_Flare
  }
  n48 = Multiply(A=@n24, B=Divide(A=@n47.RGB) {constB=9})
  n49 = CollectionParameter() {collection=&MPC_PostProcessControl, parameterName=Flare_Flicker}
  n50 = LinearInterpolate(A=@n49, Alpha=@n39) {constA=0.7}
  n51 = Add(
   A=Multiply(A=@n24, B=Add(A=@n32, B=@n43))
   B=MaterialFunctionCall(Hue Shift Percentage=Constant() {r=0.07}, Texture=Multiply(A=@n48, B=@n50)) {materialFunction=&HueShift}.Result
  )
  n52 = Multiply(A=@n35) {constB=5}
  n53 = Sine(None=@n52) {period=0.075}
  n54 = Multiply(A=Multiply(A=Sine(None=@n52) {period=0.155}, B=Sine(None=@n52) {period=0.2}), B=Sine(None=@n52) {period=0.43})
  n55 = Clamp(None=LinearInterpolate(A=@n49, Alpha=Multiply(A=@n53, B=@n54)) {constA=0.7}) {maxDefault=3}
  Add(
   A=@n23
   B=Multiply(
    A=CollectionParameter() {collection=&MPC_PostProcessControl, parameterName=Flare_Brightness}
    B=Multiply(A=@n51, B=@n55)
   )
  )
  n56 = ScalarParameter() {defaultValue=18317732, parameterName=SphereRadius}
  n57 = AppendVector(A=Constant2Vector(), B=@n56)
  n58 = Multiply(A=CameraVectorWS()) {constB=-1}
  n59 = CameraVectorWS()
  n60 = MaterialFunctionCall(Sphere Position=Add(A=Multiply(A=@n57, B=@n58), B=@n2.RGB), Sphere Radius=@n56, RayDirection=@n59, World position=WorldPosition().XYZ) {materialFunction=&RayTracedSphere}
  n61 = Subtract(A=@n60."Ray Intersection Position ", B=@n2.RGB)
  n62 = CrossProduct(A=@n61, B=@n59)
  ComponentMask(None=Normalize(VectorInput=Transform(None=@n62) {transformSourceType=TRANSFORMSOURCE_World, transformType=TRANSFORM_Camera})) {g=true, r=true}
  n63 = TextureCoordinate()
  n64 = Constant2Vector() {g=0.5, r=0.5}
  n65 = Subtract(A=@n63, B=@n64)
  n66 = Constant()
  ComponentMask(None=CrossProduct(A=Constant3Vector() {constant=<rgba 0 0 1 0>}, B=AppendVector(A=@n65, B=@n66))) {g=true, r=true}
  Constant3Vector() {constant=<rgba 1 0 0 0>}
  Distance(A=@n3, B=@n4)
  Distance(
   A=Transform(None=@n7.Intersection End World Position) {transformSourceType=TRANSFORMSOURCE_World, transformType=TRANSFORM_View}
   B=Transform(None=@n2.RGB) {transformSourceType=TRANSFORMSOURCE_World, transformType=TRANSFORM_View}
  )
  Distance(A=@n64, B=@n63)
  n67 = CameraVectorWS()
  n68 = DotProduct(A=@n7.Intersection Surface Normal, B=@n67)
  Divide(A=@n9, B=@n68)
  n69 = Frac(None=Multiply(A=Time(), B=ScalarParameter() {defaultValue=0.07, parameterName=ColapseSpeed}))
  Divide(A=@n69)
  LinearInterpolate()
  n70 = Desaturation(None=@n22)
  n71 = ScalarParameter() {defaultValue=3, parameterName=GhostEffect}
  n72 = Add(
   A=ScalarParameter() {defaultValue=5, parameterName=Test}
   B=ComponentMask(None=SceneTexture(UVs=@n18) {bFiltered=true, sceneTextureId=PPI_SceneDepth}.Color) {r=true}
  ) {constB=0.5}
  n73 = ComponentMask(None=SceneTexture(UVs=@n18) {bFiltered=true, sceneTextureId=PPI_CustomDepth}.Color) {r=true}
  LinearInterpolate(
   A=@n22
   B=Multiply(A=VectorParameter() {parameterName=ColorCharacterGhosts, defaultValue=<rgba 0.5772 3.6525 6 1>}.RGB, B=Power(Base=@n70, Exp=@n71))
   Alpha=Clamp(None=Subtract(A=@n72, B=@n73))
  )
  n74 = TextureCoordinate()
  n75 = MaterialFunctionCall(Vector or UVs=Multiply(
    A=Constant3Vector() {constant=<rgba 1 1 0 0>}
    B=Transform(None=@n61) {transformSourceType=TRANSFORMSOURCE_World, transformType=TRANSFORM_Camera}
   )) {materialFunction=&VectorToRadialValue}
  n76 = Constant() {r=1}
  n77 = ComponentMask(None=@n75.Radial Coordinates) {g=true}
  n78 = ScalarParameter() {defaultValue=1, parameterName=FallofTwist}
  n79 = Multiply(
   A=Add(A=Power(Base=@n77, Exp=@n78), B=ScalarParameter() {defaultValue=1, parameterName=Scale_Twist_Bias})
   B=ScalarParameter() {defaultValue=1, parameterName=Scale_Twist_Y}
  )
  n80 = Multiply(A=@n75.Radial Coordinates, B=AppendVector(A=ScalarParameter() {defaultValue=1, parameterName=Scale_Twist_X}, B=Divide(A=@n76, B=@n79)))
  n81 = Add(A=@n80, B=AppendVector(A=Constant(), B=ComponentMask(None=@n80) {r=true}))
  n82 = Divide(A=Distance(A=@n60."Ray Intersection Position ", B=@n2.RGB), B=ScalarParameter() {defaultValue=28605406, parameterName=RadialDistance})
  n83 = Power(Base=@n82, Exp=ScalarParameter() {defaultValue=3, parameterName=RadialFalloff})
  n84 = Add(A=@n74, B=Multiply(A=Multiply(A=@n81, B=@n83), B=ScalarParameter() {defaultValue=1.0632, parameterName=Distort}))
  n85 = Add(
   A=ScalarParameter() {defaultValue=5, parameterName=Test}
   B=ComponentMask(None=SceneTexture() {bFiltered=true, sceneTextureId=PPI_SceneDepth}.Color) {r=true}
  ) {constB=0.5}
  n86 = ComponentMask(None=SceneTexture() {bFiltered=true, sceneTextureId=PPI_CustomDepth}.Color) {r=true}
  LinearInterpolate(
   A=@n23
   B=ComponentMask(None=SceneTexture(UVs=@n84) {sceneTextureId=PPI_PostProcessInput0}.Color) {b=true, g=true, r=true}
   Alpha=Clamp(None=Subtract(A=@n85, B=@n86))
  )
  n87 = Constant3Vector() {constant=<rgba 0 0 1 0>}
  n88 = ScalarParameter() {defaultValue=50000, parameterName=Scale}
  MaterialFunctionCall(
   Plane Normal Axis=@n87
   Plane X Axis=Normalize(VectorInput=CrossProduct(A=@n87, B=@n67))
   Plane Center=@n2.RGB
   UV Scale=AppendVector(A=@n88, B=@n88)
  ) {materialFunction=&VirtualPlaneCoordinates}
  MaterialFunctionCall!(Plane Normal Axis, Plane X Axis, Plane Center, UV Scale) {materialFunction=/Game/Athena/Environments/Nexus/Materials/VirtualPlaneCoordinatesAbs}
  Multiply(A=@n68, B=@n8)
  Multiply() {constB=-1}
  OneMinus(None=@n82)
  OneMinus(None=@n13)
  n89 = ScalarParameter() {defaultValue=1.1, parameterName=PauseAdd}
  Reroute(None=Saturate(None=Multiply(A=@n69, B=@n89) {constB=2}))
  RotateAboutAxis(Position=@n74)
  Rotator(Coordinate=@n46.UVs) {speed=0.08}
  n90 = Distance(A=@n1, B=@n12)
  n91 = ScalarParameter() {defaultValue=-0.15, parameterName=FallofRadiusAOffset}
  Saturate(None=Divide(A=Add(A=@n90, B=@n91), B=ScalarParameter() {defaultValue=0.5, parameterName=FallofRadiusA}))
  ScalarParameter() {defaultValue=1, parameterName=Progress}
  TextureCoordinate()
  TextureCoordinate()
  TextureCoordinate()
 ]
}
