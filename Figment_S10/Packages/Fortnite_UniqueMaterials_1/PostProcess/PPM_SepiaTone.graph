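# Post-process material, applied after depth of field, that gives the scene a sepia, dirty film look.
# Scene color goes through a sepia matrix built from masked multiplies and adds, then blends with the original by SepiaTone.
# Time through Ceil and Frac makes the film-dirt texture's UVs jump in steps.
# Its red channel darkens the image, with strength varied by the same texture's unshifted green channel.
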Material /Figment_S10/Packages/Fortnite_UniqueMaterials_1/PostProcess/PPM_SepiaTone
graph {
 properties {blendableLocation=BL_SceneColorAfterDOF, materialDomain=MD_PostProcess, pixelDepthOffsetMode=PDOM_Legacy, usedShadingModels=Default Lit}
 nodes [
  n1 = ComponentMask(None=SceneTexture() {sceneTextureId=PPI_PostProcessInput0}.Color) {b=true, g=true, r=true}
  n2 = Multiply(A=Constant3Vector() {constant=<rgba 0.393 0.769 0.189 0>}, B=@n1)
  n3 = ComponentMask(None=@n2) {r=true}
  n4 = ComponentMask(None=@n2) {g=true}
  n5 = Multiply(A=Constant3Vector() {constant=<rgba 0.349 0.686 0.168 0>}, B=@n1)
  n6 = ComponentMask(None=@n5) {r=true}
  n7 = ComponentMask(None=@n5) {g=true}
  n8 = AppendVector(A=Add(A=Add(A=@n3, B=@n4), B=ComponentMask(None=@n2) {b=true}), B=Add(A=Add(A=@n6, B=@n7), B=ComponentMask(None=@n5) {b=true}))
  n9 = Multiply(A=Constant3Vector() {constant=<rgba 0.272 0.534 0.131 0>}, B=@n1)
  n10 = Add(A=Add(A=ComponentMask(None=@n9) {r=true}, B=ComponentMask(None=@n9) {g=true}), B=ComponentMask(None=@n9) {b=true})
  n11 = Time()
  n12 = ScalarParameter() {defaultValue=10, parameterName=SpeedUVShiftDirt}
  n13 = Multiply(A=Ceil(None=Multiply(A=@n11, B=@n12)), B=ScalarParameter() {defaultValue=1.3425, parameterName=UVShiftoffsetDirt})
  n14 = ScalarParameter() {defaultValue=12, parameterName=SpeedUVShiftDirt2}
  n15 = Multiply(A=Ceil(None=Multiply(A=@n11, B=@n14)), B=ScalarParameter() {defaultValue=1.3425, parameterName=UVShiftoffsetDirt2})
  n16 = Add(
   A=AppendVector(A=Frac(None=@n13), B=Frac(None=@n15))
   B=Multiply(A=TextureCoordinate(), B=ScalarParameter() {defaultValue=0.75, parameterName=DirtScale})
  )
  n17 = TextureSample() {texture=/Game/Environments/Props/DriveIn/Textures/T_FilmLensDirt}
  n18 = LinearInterpolate(A=ScalarParameter() {defaultValue=0.18, parameterName=DirtBrightness}, B=ScalarParameter() {parameterName=DirtBrightness2}, Alpha=@n17.G)
  n19 = OneMinus(None=Multiply(A=TextureSample(UVs=@n16) {texture=&T_FilmLensDirt}.R, B=Saturate(None=@n18)))
  Multiply(A=LinearInterpolate(A=@n1, B=AppendVector(A=@n8, B=@n10), Alpha=ScalarParameter() {defaultValue=0.4381, parameterName=SepiaTone}), B=@n19)
  Multiply(A=@n19, B=@n17.G)
 ]
}
